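# Physics asset for the face accessory of the male medium BronzeHat cosmetic, setting up physics for its dynamic hood bones.
# A three-link back chain and shorter, tighter-limited front link hang from BASE bodies constrained to spine_04.
# Capsules on pelvis, spine, left thigh and left upper arm supply body shapes, with no right-side limbs.
# Body and BASE joints leave swing and twist free, while hood links carry angular limits.
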PhysicsAsset /BRCosmetics/Characters/Player/Male/Medium/Bodies/M_MED_BronzeHat/Meshes/Parts/M_MED_BronzeHat_FaceAcc_Physics
bodies {
 dyn_hood_back_01 [
  {shapeType=Capsule, length=19.6491, radius=1.7948, shapeName=dyn_hood_back_01_capsule, center=<5 0 0>, rotation=<rot 0 0 90>}
  {shapeType=Capsule, length=17.182, radius=2, shapeName=dyn_hood_back_01_capsule, rotation=<rot 0 0 90>}
 ]
 dyn_hood_back_02 [{shapeType=Capsule, length=16.7575, radius=1.7948, shapeName=dyn_hood_back_02_capsule, center=<5 0 0>, rotation=<rot 0 0 90>}]
 dyn_hood_back_03 [{shapeType=Capsule, length=10.0631, radius=1.7948, shapeName=dyn_hood_back_03_capsule, center=<7 0 0>, rotation=<rot 0 0 90>}]
 dyn_hood_back_BASE [{shapeType=Capsule, length=1, radius=1, shapeName=dyn_hood_back_BASE_capsule}]
 dyn_hood_front_01 [{shapeType=Capsule, length=3.3559, radius=2.3614, shapeName=dyn_hood_front_01_capsule, center=<5 0 0>}]
 dyn_hood_front_BASE [{shapeType=Capsule, length=1, radius=1, shapeName=dyn_hood_front_BASE_capsule}]
 pelvis [{shapeType=Capsule, length=18.7199, radius=12, shapeName=pelvis_capsule}]
 spine_01 [{shapeType=Capsule, length=8.5661, radius=11.7278, shapeName=spine_01_capsule}]
 spine_02 [{shapeType=Capsule, length=8.5661, radius=11.7278, shapeName=spine_02_capsule}]
 spine_03 [{shapeType=Capsule, length=8.5661, radius=11.7278, shapeName=spine_03_capsule}]
 spine_04 [{shapeType=Capsule, length=8, radius=12, shapeName=spine_04_capsule, rotation=<rot 0 -90 0>}]
 thigh_l [{shapeType=Capsule, length=31.3827, radius=7.8866, shapeName=thigh_l_capsule, center=<-21.4091 0 0>, rotation=<rot 0 -90 90>}]
 upperarm_l [{shapeType=Capsule, length=19.9568, radius=6.2362, shapeName=upperarm_l_capsule, center=<9.3511 0 0>, rotation=<rot 90 0 0>}]
}
constraints [
 {
  bone1Name=spine_03
  bone2Name=spine_02
  swing1LimitDegrees=45
  swing1Motion=Free
  swing2LimitDegrees=45
  swing2Motion=Free
  twistLimitDegrees=45
  twistMotion=Free
 }
 {
  bone1Name=spine_02
  bone2Name=spine_01
  swing1LimitDegrees=45
  swing1Motion=Free
  swing2LimitDegrees=45
  swing2Motion=Free
  twistLimitDegrees=45
  twistMotion=Free
 }
 {
  bone1Name=spine_01
  bone2Name=pelvis
  swing1LimitDegrees=45
  swing1Motion=Free
  swing2LimitDegrees=45
  swing2Motion=Free
  twistLimitDegrees=45
  twistMotion=Free
 }
 {
  bone1Name=dyn_hood_back_BASE
  bone2Name=spine_04
  swing1LimitDegrees=45
  swing1Motion=Free
  swing2LimitDegrees=45
  swing2Motion=Free
  twistLimitDegrees=45
  twistMotion=Free
 }
 {bone1Name=dyn_hood_back_01, bone2Name=dyn_hood_back_BASE, swing1LimitDegrees=45, swing2LimitDegrees=40, twistLimitDegrees=25}
 {bone1Name=dyn_hood_back_02, bone2Name=dyn_hood_back_01, swing1LimitDegrees=45, swing2LimitDegrees=45, twistLimitDegrees=25}
 {bone1Name=dyn_hood_back_03, bone2Name=dyn_hood_back_02, swing1LimitDegrees=50, swing2LimitDegrees=50, twistLimitDegrees=25}
 {
  bone1Name=dyn_hood_front_BASE
  bone2Name=spine_04
  swing1LimitDegrees=45
  swing1Motion=Free
  swing2LimitDegrees=45
  swing2Motion=Free
  twistLimitDegrees=45
  twistMotion=Free
 }
 {bone1Name=dyn_hood_front_01, bone2Name=dyn_hood_front_BASE, swing1LimitDegrees=25, swing2LimitDegrees=25, twistLimitDegrees=10}
 {
  bone1Name=upperarm_l
  bone2Name=spine_04
  swing1LimitDegrees=45
  swing1Motion=Free
  swing2LimitDegrees=45
  swing2Motion=Free
  twistLimitDegrees=45
  twistMotion=Free
 }
 {
  bone1Name=thigh_l
  bone2Name=pelvis
  swing1LimitDegrees=45
  swing1Motion=Free
  swing2LimitDegrees=45
  swing2Motion=Free
  twistLimitDegrees=45
  twistMotion=Free
 }
 {
  bone1Name=spine_04
  bone2Name=spine_03
  swing1LimitDegrees=45
  swing1Motion=Free
  swing2LimitDegrees=45
  swing2Motion=Free
  twistLimitDegrees=45
  twistMotion=Free
 }
]
