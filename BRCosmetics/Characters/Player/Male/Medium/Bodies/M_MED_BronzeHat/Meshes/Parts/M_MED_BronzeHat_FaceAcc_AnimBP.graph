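# Post-process style part AnimBP for the BronzeHat cosmetic FaceAcc mesh, adding secondary motion and a floating crown.
# Copied parent and owner-mesh poses pass through a deform Control Rig into a cached pose, feeding two RigidBody sims blended by a cape/long-backbling check.
# The crown base then gets an offset and real-time sine bob before AnimDynamics on dyn_crown_01.
# Hood dynamics and bone-constraint chains are disconnected from the output.
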AnimBlueprint /BRCosmetics/Characters/Player/Male/Medium/Bodies/M_MED_BronzeHat/Meshes/Parts/M_MED_BronzeHat_FaceAcc_AnimBP
target_skeleton=&/M_MED_BronzeHat_FaceAcc_Skeleton
graphs [
 AnimGraph {nodes [
   n1 = Anim Dynamics - Bone: dyn_hood_r!(ComponentPose, Alpha=1, GravityOverride:Vector="0,0,0", ExternalForce="0,0,0") -> (Pose) {
    angularDampingOverride=0.999
    angularSpringConstant=500
    bAngularSpring=true
    bGravityOverrideInSimSpace=true
    bLinearSpring=true
    bOverrideAngularDamping=true
    bOverrideLinearDamping=true
    bUseGravityOverride=true
    gravityScale=0.3
    linearDampingOverride=0.999
    linearSpringConstant=200
    simulationSpace=RootRelative
    alphaBoolBlend {blendOption=Linear}
    boundBone {boneName=dyn_hood_r}
    componentAppliedLinearAccClamp=<1000 1000 1000>
    componentLinearAccScale=<1 1 1>
    componentLinearVelScale=<1 1 1>
    physicsBodyDefinitions [{
      sphereCollisionRadius=0
      boundBone {boneName=dyn_hood_r}
      boxExtents=<4 8 8>
      constraintSetup {
       angularTargetAxis=AxisY
       angularLimitsMax=<3 15 3>
       angularLimitsMin=<-3 -15 -3>
       angularTarget=<0 0 0>
       linearAxesMax=<0.3 0.3 0.3>
       linearAxesMin=<-0.3 -0.3 -0.3>
      }
      localJointOffset=<2 0 0>
     }]
   }
   n2 = Anim Dynamics - Bone: dyn_hood_l!(ComponentPose=@n1, Alpha=0.95, GravityOverride:Vector="0,0,0", ExternalForce="0,0,0") -> (Pose) {
    angularDampingOverride=0.999
    angularSpringConstant=500
    bAngularSpring=true
    bGravityOverrideInSimSpace=true
    bLinearSpring=true
    bOverrideAngularDamping=true
    bOverrideLinearDamping=true
    bUseGravityOverride=true
    gravityScale=0.3
    linearDampingOverride=0.999
    linearSpringConstant=200
    simulationSpace=RootRelative
    alphaBoolBlend {blendOption=Linear}
    boundBone {boneName=dyn_hood_l}
    componentAppliedLinearAccClamp=<1000 1000 1000>
    componentLinearAccScale=<1 1 1>
    componentLinearVelScale=<1 1 1>
    physicsBodyDefinitions [{
      sphereCollisionRadius=0
      boundBone {boneName=dyn_hood_l}
      boxExtents=<4 8 8>
      constraintSetup {
       angularTargetAxis=AxisY
       angularLimitsMax=<3 15 3>
       angularLimitsMin=<-3 -15 -3>
       angularTarget=<0 0 0>
       linearAxesMax=<0.3 0.3 0.3>
       linearAxesMin=<-0.3 -0.3 -0.3>
      }
      localJointOffset=<2 0 0>
     }]
   }
   Anim Dynamics - Bone: dyn_hood_top!(ComponentPose=@n2, Alpha=0.6, GravityOverride:Vector="0,0,0", ExternalForce="0,0,0") -> (Pose) {
    angularDampingOverride=0.999
    angularSpringConstant=500
    bAngularSpring=true
    bGravityOverrideInSimSpace=true
    bLinearSpring=true
    bOverrideAngularDamping=true
    bOverrideLinearDamping=true
    bUseGravityOverride=true
    gravityScale=0.3
    linearDampingOverride=0.999
    linearSpringConstant=200
    simulationSpace=RootRelative
    alphaBoolBlend {blendOption=Linear}
    boundBone {boneName=dyn_hood_top}
    componentAppliedLinearAccClamp=<1000 1000 1000>
    componentLinearAccScale=<1 1 1>
    componentLinearVelScale=<1 1 1>
    physicsBodyDefinitions [{
      sphereCollisionRadius=0
      boundBone {boneName=dyn_hood_top}
      boxExtents=<4 8 8>
      constraintSetup {
       angularTargetAxis=AxisY
       angularLimitsMax=<3 3 15>
       angularLimitsMin=<-3 -3 -15>
       angularTarget=<0 0 0>
       linearAxesMax=<0.3 0.2 0.3>
       linearAxesMin=<-0.3 -0.2 -0.3>
      }
      localJointOffset=<2 0 0>
     }]
   }
   n3 = "Constraint - dyn_hood_back_BASE (PPPPPPP)"!(
    ComponentPose
    Alpha=1
    ConstraintWeights_0=0.082
    ConstraintWeights_1:float=0.061
    ConstraintWeights_2:float=0.072
    ConstraintWeights_3:float=0.11
    ConstraintWeights_4:float=0.369
    ConstraintWeights_5:float=0.198
    ConstraintWeights_6:float=0.108
   ) -> (Pose) {
    alphaBoolBlend {blendOption=Linear}
    boneToModify {boneName=dyn_hood_back_BASE}
    constraintSetup [
     {targetBone {boneName=spine_03}}
     {targetBone {boneName=spine_04}}
     {targetBone {boneName=clavicle_l}}
     {targetBone {boneName=upperarm_l}}
     {targetBone {boneName=upperarm_twist_01_l}}
     {targetBone {boneName=deform_upperarm_bk_l}}
     {targetBone {boneName=upperarm_twist_02_l}}
    ]
    constraintWeights [1,1,1,1,1,1,1]
   }
   n4 = "Constraint - dyn_hood_front_BASE (PPPPPPP)"!(
    ComponentPose=@n3
    Alpha=1
    ConstraintWeights_0=0.076
    ConstraintWeights_1:float=0.036
    ConstraintWeights_2:float=0.069
    ConstraintWeights_3:float=0.356
    ConstraintWeights_4:float=0.201
    ConstraintWeights_5:float=0.112
    ConstraintWeights_6:float=0.15
   ) -> (Pose) {
    alphaBoolBlend {blendOption=Linear}
    boneToModify {boneName=dyn_hood_front_BASE}
    constraintSetup [
     {targetBone {boneName=spine_04}}
     {targetBone {boneName=spine_05}}
     {targetBone {boneName=deform_pec_l}}
     {targetBone {boneName=upperarm_twist_01_l}}
     {targetBone {boneName=deform_upperarm_fr_l}}
     {targetBone {boneName=upperarm_twist_02_l}}
     {targetBone {boneName=clavicle_l}}
    ]
    constraintWeights [1,1,1,1,1,1,1]
   }
   "Constraint - extra_button (PPPPP)"!(
    ComponentPose=@n4
    Alpha=1
    ConstraintWeights_0=0.08
    ConstraintWeights_1:float=0.235
    ConstraintWeights_2:float=0.377
    ConstraintWeights_3:float=0.104
    ConstraintWeights_4:float=0.204
   ) -> (Pose) {
    alphaBoolBlend {blendOption=Linear}
    boneToModify {boneName=extra_button}
    constraintSetup [
     {targetBone {boneName=spine_03}}
     {targetBone {boneName=spine_04}}
     {targetBone {boneName=spine_05}}
     {targetBone {boneName=deform_pec_r}}
     {targetBone {boneName=clavicle_r}}
    ]
    constraintWeights [1,1,1,1,1]
   }
   n5 = RigidBody!(ComponentPose=Local To Component(LocalPose=Use cached pose 'StartPose'()), Alpha, OverrideWorldGravity:Vector="0,0,0", ExternalForce) -> (Pose) {
    bClampLinearTranslationLimitToRefPose=true
    bOverrideWorldGravity=true
    overridePhysicsAsset=&/M_MED_BronzeHat_FaceAcc_Physics
    simulationSpace=BaseBoneSpace
    alphaBoolBlend {blendOption=Linear}
    baseBoneRef {boneName=spine_04}
    componentAppliedLinearAccClamp=<3000 3000 3000>
    simSpaceSettings {
     maxAngularAcceleration=12
     maxAngularVelocity=4
     maxLinearAcceleration=1050
     maxLinearVelocity=350
     worldAlpha=1
     externalLinearDragV=<1 1 2>
     externalLinearVelocity {x=0.1}
    }
   }
   n6 = Component To Local(ComponentPose=@n5)
   n7 = RigidBody(ComponentPose=Local To Component(LocalPose=Use cached pose 'StartPose'()), OverrideWorldGravity="0,0,0") {
    bClampLinearTranslationLimitToRefPose=true
    bOverrideWorldGravity=true
    overridePhysicsAsset=&/M_MED_BronzeHat_FaceAcc_Physics_WCape
    simulationSpace=BaseBoneSpace
    alphaBoolBlend {blendOption=Linear}
    baseBoneRef {boneName=spine_04}
    componentAppliedLinearAccClamp=<3000 3000 3000>
    simSpaceSettings {
     maxAngularAcceleration=12
     maxAngularVelocity=4
     maxLinearAcceleration=1050
     maxLinearVelocity=350
     worldAlpha=1
     externalLinearDragV=<1 1 2>
     externalLinearVelocity {x=0.1}
    }
   }
   n8 = Component To Local(ComponentPose=@n7)
   n9 = Get WithCape() -> (WithCape:bool)
   n10 = "Transform (Modify) Bone - Bone: dyn_crown_BASE"(
    ComponentPose=Local To Component(LocalPose=Blend Poses by bool(BlendPose_0=@n8, BlendPose_1=@n6, bActiveValue=@n9) {blendType=Linear})
    Alpha=1
    Translation="0,0,1.3"
    Rotation="0,0,0"
    Scale="1,1,1"
   ) {translationMode=BMM_Additive, alphaBoolBlend {blendOption=Linear}, boneToModify {boneName=dyn_crown_BASE}, translation {z=1.3}}
   n11 = "Transform (Modify) Bone - Bone: dyn_crown_base"(ComponentPose=@n10, Alpha=1, Translation=Get Crown_Float() -> (Crown_Float:Vector), Rotation="0,0,0", Scale="1,1,1") {translationMode=BMM_Additive, translationSpace=BCS_BoneSpace, alphaBoolBlend {blendOption=Linear}, boneToModify {boneName=dyn_crown_base}}
   n12 = Anim Dynamics - Bone: dyn_crown_01!(ComponentPose=@n11, Alpha=1, GravityOverride:Vector="0,0,0", ExternalForce="0,0,0") -> (Pose) {
    alpha=0
    angularBiasOverride=1
    angularDampingOverride=1
    angularSpringConstant=20
    bGravityOverrideInSimSpace=true
    bLinearSpring=true
    bOverrideAngularBias=true
    bOverrideAngularDamping=true
    bOverrideLinearDamping=true
    bUseGravityOverride=true
    bUsePlanarLimit=false
    gravityScale=0
    linearDampingOverride=1
    linearSpringConstant=200
    simulationSpace=RootRelative
    alphaBoolBlend {blendOption=Linear}
    alphaScaleBiasClamp {scale=0.2}
    boundBone {boneName=dyn_crown_01}
    componentAppliedLinearAccClamp=<10 10 10>
    componentLinearAccScale=<0.1 0.1 0.1>
    componentLinearVelScale=<0.1 0.1 0.1>
    physicsBodyDefinitions [{
      sphereCollisionRadius=0
      boundBone {boneName=dyn_crown_01}
      boxExtents=<5 5 5>
      constraintSetup {
       linearXLimitType=Free
       linearZLimitType=Free
       angularLimitsMax=<0 0 0>
       angularLimitsMin=<0 0 0>
       angularTarget=<0 0 0>
       linearAxesMax=<2 0.5 2>
       linearAxesMin=<-2 -8 -2>
      }
      localJointOffset=<0 0 0>
     }]
    relativeSpaceBone {boneName=root}
   }
   Output Pose(Result=Component To Local(ComponentPose=@n12)) {layerGroup=None}
   n13 = Layered blend per bone(
    BasePose=Copy Pose From Mesh() {bUseAttachedParent=true}
    BlendPoses_0=Copy Pose From Mesh(SourceMeshComponent=Get Skel_Mesh()) {bCopyCurves=true, bUseAttachedParent=true}
   ) {layerSetup [{branchFilters [
       {boneName=clavicle_l}
       {boneName=upperarm_l}
       {boneName=upperarm_twist_01_l}
       {boneName=upperarm_twist_02_l}
       {boneName=clavicle_r}
       {boneName=upperarm_r}
       {boneName=upperarm_twist_01_r}
       {boneName=upperarm_twist_02_r}
       {boneName=thigh_l}
      ]}]}
   n14 = Control Rig(Source=@n13) {
    alphaBoolBlend {blendOption=Linear}
    controlRigAssetReference {blueprintRigClass=&/M_MED_BronzeHat_FaceAcc_CtrlRig_Deform.M_MED_BronzeHat_FaceAcc_CtrlRig_Deform_C}
   }
   Save cached pose 'StartPose'(Pose=@n14) {cacheName=StartPose}
  ]}
 EventGraph {nodes [
   Event BlueprintUpdateAnimation()
   > n1 = Sequence!(execute) -> (then_0, then_1, then_2:Exec, then_3:Exec)
   Set Skel_Mesh(execute=@n1.then_1, Skel_Mesh=GetOwnerSkeletalMeshComponent())
   n2 = Gravity Override State Function DEPRECATED(
    execute=@n1.then_0
    CharacterPartAnimInstance=Self-Reference()
    GameTimeInSeconds=GetGameTimeInSeconds()
    GravityOverrideParams=Get Gravity Override Params()
   )
   > Set GravityOverride(GravityOverride=@n2)
   n3 = GetRealTimeSeconds()
   n4 = "Sin (Radians)"(A=float * float(A=float * float(A=1, B=1), B=@n3))
   n5 = "Sin (Radians)"(A=float * float(A=float * float(A=1, B=1), B=@n3))
   n6 = "Sin (Radians)"(A=float * float(A=float * float(A=1, B=2.3), B=@n3))
   Set Crown_Float(
    execute=@n1.then_2
    Crown_Float:Vector=MakeVector(X=float * float(A=@n6, B=0.9), Y=float * float(A=@n4, B=0.2), Z=float * float(A=@n5, B=0.15))
   ) -> (Output_Get:Vector)
   n7 = CheckForBackbling()
   Set WithCape(execute=@n1.then_3, WithCape:bool=OR Boolean(A=@n7.bIsBackblingACape, B=@n7.bIsBackblingLong)) -> (Output_Get:bool)
  ]}
]
variables [
 Crown_Float {type=Vector}
 Gravity Override Params {type=GravityOverrideParamsStruct}
 GravityOverride {type=Vector}
 Skel_Mesh {type=SkeletalMeshComponent}
 WithCape {type=boolean}
]
